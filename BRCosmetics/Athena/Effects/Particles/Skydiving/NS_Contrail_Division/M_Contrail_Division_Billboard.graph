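# Unlit translucent billboard material for Niagara mesh particles in the Division skydiving contrail effect.
# The Sprite function billboards the mesh, sized by a Scale parameter.
# A Mask texture, defaulting to a soft flare, multiplies particle color in its green channel and feeds the emissive adjustments function.
# A selectable mask channel, green by default, multiplies particle alpha for transparency.
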Material /BRCosmetics/Athena/Effects/Particles/Skydiving/NS_Contrail_Division/M_Contrail_Division_Billboard
graph {
 properties {
  bUsedWithNiagaraMeshParticles=true
  blendMode=BLEND_Translucent
  pixelDepthOffsetMode=PDOM_Legacy
  shadingModel=MSM_Unlit
  usedShadingModels=Unlit
 }
 nodes [
  n1 = ParticleColor()
  n2 = TextureSampleParameter2D() {parameterName=Mask, texture=/Game/Effects/Fort_Effects/Textures/Flares/T_Flare_Soft}
  MaterialFunctionCall(Emissive=Multiply(A=@n1.RGB, B=@n2.G)) {materialFunction=&MF_FX_EmissiveAdjustments}
  MaterialFunctionCall!(
   Center Location=WorldPosition().XYZ
   0-1 UVs
   XY scale=Multiply(A=ScalarParameter() {defaultValue=1, parameterName=Scale}, B=Constant2Vector() {g=30, r=30})
   OverridePivot
   "Normals (Optional)"
  ) {materialFunction=&Sprite}
  Multiply(A=StaticComponentMaskParameter(None=@n2.RGBA) {defaultG=true, parameterName=Masked Channel}, B=@n1.A)
 ]
}
