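# Translucent, unlit, two-sided spline material for the Division skydiving contrail, used on Niagara mesh particles and spline meshes.
# A U-coordinate gradient forms a soft band with a sharp center core, driving opacity and a brighter particle-colored emissive passed through MF_FX_EmissiveAdjustments.
# Remapped particle alpha fades the core glow and opacity and narrows the strand.
# SplineThicken sets width from that alpha, the SplineWidth parameter and slight per-particle random variation.
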Material /BRCosmetics/Athena/Effects/Particles/Skydiving/NS_Contrail_Division/M_Contrail_Division_Spline
graph {
 properties {
  bCastRayTracedShadows=false
  bUsedWithNiagaraMeshParticles=true
  bUsedWithSplineMeshes=true
  blendMode=BLEND_Translucent
  pixelDepthOffsetMode=PDOM_Legacy
  shadingModel=MSM_Unlit
  twoSided=true
  usedShadingModels=Unlit
 }
 nodes [
  n1 = ComponentMask(None=TextureCoordinate()) {r=true}
  n2 = Multiply(A=OneMinus(None=@n1), B=@n1)
  n3 = Power(Base=Multiply(A=@n2) {constB=4}) {constExponent=15}
  n4 = ParticleColor()
  n5 = Constant() {r=0.6}
  n6 = Add(A=@n5)
  n7 = Saturate(None=Subtract(A=Multiply(A=@n4.A, B=@n6), B=@n5))
  n8 = Multiply(A=@n3, B=@n7)
  MaterialFunctionCall(Emissive=Multiply(A=LinearInterpolate(Alpha=@n8) {constA=0.2}, B=@n4.RGB)) {materialFunction=&MF_FX_EmissiveAdjustments}
  n9 = ParticleRandom()
  n10 = Multiply(
   A=LinearInterpolate(Alpha=@n7) {constA=0.2}
   B=Multiply(A=ScalarParameter() {defaultValue=1, parameterName=SplineWidth}, B=LinearInterpolate(Alpha=@n9) {constA=1.75, constB=2})
  ) {constA=1}
  MaterialFunctionCall(WidthBase=@n10, WidthTip=@n10) {materialFunction=&SplineThicken}
  n11 = Add(A=@n3, B=@n2)
  Multiply(A=Multiply(B=Saturate(None=@n11)) {constA=1}, B=@n7)
 ]
}
