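# NS_Contrail_Division is a Niagara skydiving contrail cosmetic with a constellation and star theme, trailing from two points on the player's skeletal mesh.
# Hand emitters layer constellation meshes with ribbons, smoke, sprite streaks and lights, splitting ColorA and ColorB by spawn group, plus a one-shot birth burst.
# Stars spawn on a surrounding cylinder and move opposite the dive vector.
# Diving attributes and a system integrator drive orientation, offsets, spawn rates, lifetimes and stretching.
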NiagaraSystem /BRCosmetics/Athena/Effects/Particles/Skydiving/NS_Contrail_Division/NS_Contrail_Division
dependencies {
 usedDataInterfaces [
  &NiagaraDataInterfaceMeshRendererInfo
  &NiagaraDataInterfaceCurve
  &NiagaraDataInterfaceVectorCurve
  &NiagaraDataInterfaceVector2DCurve
  &NiagaraDataInterfaceArrayInt32
  &NiagaraDataInterfaceColorCurve
 ]
 usedDynamicInputs [
  &SetBoolByFloatComparison
  &FN_FrontEndVector3Modifier
  &self:Abs_Z_Axis
  &TransformVector
  &Multiply_Float
  &Multiply_Vector
 ]
 usedModules [
  &SystemState
  /Game/Effects/Niagara/Modules/Skydiving/FN_DivingAttributes/v5/FN_DivingAttributes
  &self:SystemSpawnScript.SystemScriptSource.SystemScriptGraph.NiagaraNodeAssignment_2.SetVariables_#1
  &FN_Integrator
  &self:SystemSpawnScript.SystemScriptSource.SystemScriptGraph.NiagaraNodeAssignment_1.SetVariables_#2
  &self:SystemSpawnScript.SystemScriptSource.SystemScriptGraph.NiagaraNodeAssignment_4.SetVariables_#3
  &self:SystemSpawnScript.SystemScriptSource.SystemScriptGraph.NiagaraNodeAssignment_0.SetVariables_#4
 ]
}
emitters {
 Constellation_Hands {
  emitterUpdateScript [
   &EmitterState(Delay First Loop Only=true, Life Cycle Mode=::NewEnumerator1, Loop Delay=0.33, Loop Duration=1, UseLoopDelay=true)
   &SpawnRate(SpawnRate=&Lerp_Float(A=9, B=18, Alpha=$System.FN_Integrator001.Value), Use Spawn Probability=true)
   &SpawnBurst_Instantaneous(Spawn Count=2, Use Spawn Probability=false, UseLoopCountLimit=true)
  ]
  particleSpawnScript [
   &FN_SetSpawnGroup(Number of Spawn Groups=2)
   &InitializeParticle(
    Color=&LerpLinearColors(
     StartColor=$System.ColorA
     EndColor=$System.ColorB
     LerpFactor=&MakeFloatFromInt(Int=$Particles.SpawnGroup)
    )
    Lifetime=&Lerp_Float(A=1, B=0.6, Alpha=$System.FN_Integrator001.Value)
    Mesh Scale=&MakeVector(
     X=&RandomRangeFloat(Minimum=0.3, Maximum=1)
     Y=&RandomRangeFloat(Minimum=0.3, Maximum=1)
     Z=&RandomRangeFloat(Minimum=0.5, Maximum=1.25)
    )
    Mesh Scale Mode=::NewEnumerator4
    UsePositionOffset=true
    Write Material Random=true
   )
   &InitialMeshOrientation(
    Rotation=&MakeVector(
     X=&RandomRangeFloat(Minimum=-0.1, Maximum=0.1)
     Y=&RandomRangeFloat(Minimum=-0.1, Maximum=0.1)
     Z=&RandomRangeFloat(Minimum=0, Maximum=1)
    )
    Use Orientation Vector=false
    Use Rotation Vector=true
   )
  ]
  particleUpdateScript [
   &ParticleState()
   /SetVariables(Particles.MeshOrientation=&MultiplyQuaternion(
     Quaternion A=&MultiplyQuaternion(
      Quaternion A=&FindQuatBetweenTwoVectors(From Vector=<0 0 -1>, To Vector=$System.DiveVelocityScale)
      Quaternion B=&FindQuatBetweenTwoVectors(From Vector=<1 0 0>, To Vector=$Engine.Owner.SystemYAxis)
     )
     Quaternion B=&MakeQuatFromAxisAngle(
      Axis=<0 0 1>
      AngleInDegrees=&Add_Float(
       A=&RandomRangeFloat(Minimum=-180, Maximum=180)
       B=&Lerp_Float(A=0, B=&RandomRangeFloat(Minimum=-90, Maximum=90), Alpha=$Particles.NormalizedAge)
      )
     )
    )) {isSetParameters=true, module=SetVariables_#5}
   &SkeletalMeshLocation(
    Bone / Socket Index=$Particles.SpawnGroup
    Bone Sampling Mode=::NewEnumerator4
    Inherit Velocity=false
    Interpolate Mesh Sampling=false
    MaskSpawnGroup=false
    Sampled Position Offset=&Add_Vector(
     A=&Multiply_VectorByFloat(
      Vector=$System.DiveVelocityScale
      Float=&FloatFromCurve(
       FloatCurve=DataInterface() {propertyValues {Curve {keys [
           {arriveTangent=0.3699, interpMode=RCIM_Cubic, leaveTangent=0.3699, tangentMode=RCTM_User, time=0, value=-0.1}
           {arriveTangent=1.573, interpMode=RCIM_Cubic, leaveTangent=1.573, tangentMode=RCTM_User, time=1, value=1}
          ]}}}
       Scale Curve=$System.DiveOffset
      )
     )
     B=&RandomRangeVector(Minimum=<-10 -10 2>, Maximum=<10 10 2>)
    )
    Skeletal Mesh=$User.SkeletalMesh
   )
   /SetVariables(Particles.Position=&Add_Vector(A=Linked() {linkedVariable {name=Particles.Position, type=NiagaraPosition}}, B=<0 0 0>)) {enabled=false, isSetParameters=true, module=SetVariables_#6}
   &ScaleMeshSize(Scale Factor=&VectorFromCurve(
     VectorCurve=DataInterface() {propertyValues {
       XCurve {keys [
         {arriveTangent=4.0207, interpMode=RCIM_Cubic, leaveTangent=4.0207, tangentMode=RCTM_User, time=0, value=0.5}
         {interpMode=RCIM_Cubic, tangentMode=RCTM_User, time=0.25, value=1}
         {arriveTangent=-1.2777, interpMode=RCIM_Cubic, leaveTangent=-1.2777, tangentMode=RCTM_User, time=1, value=0.3}
        ]}
       YCurve {keys [
         {arriveTangent=4.0207, interpMode=RCIM_Cubic, leaveTangent=4.0207, tangentMode=RCTM_User, time=0, value=0.5}
         {interpMode=RCIM_Cubic, tangentMode=RCTM_User, time=0.25, value=1}
         {arriveTangent=-1.2777, interpMode=RCIM_Cubic, leaveTangent=-1.2777, tangentMode=RCTM_User, time=1, value=0.3}
        ]}
       ZCurve {keys [<at 0.1 0.4 RCIM_Cubic>,<at 1 1 RCIM_Cubic>]}
      }}
     Scale Curve=&Lerp_Vector(A=<1 1 1>, B=<0.6 0.6 1.2>, Alpha=$System.FN_Integrator001.Value)
    ))
   &ScaleColor(
    Scale Alpha=&FloatFromCurve(FloatCurve=DataInterface() {propertyValues {Curve {keys [<at 0 0 RCIM_Cubic>,<at 0.15 1>,<at 0.6 1 RCIM_Cubic>,<at 1 0 RCIM_Cubic>]}}})
    Scale RGB=&Multiply_VectorByFloat(
     Vector=<5 5 5>
     Float=&FloatFromCurve(FloatCurve=DataInterface() {propertyValues {Curve {keys [<at 0 0.3 RCIM_Cubic>,<at 0.15 1>,<at 0.6 1 RCIM_Cubic>,<at 1 0.3 RCIM_Cubic>]}}})
    )
    ScaleA=true
    ScaleRGB=true
   )
   &DynamicMaterialParameters(
    Index 1 Param 2=4
    Module.Param0WriteEnabled1=true
    Module.Param1WriteEnabled1=true
    Module.Param2WriteEnabled1=false
    Module.Param3WriteEnabled1=false
    Write Parameter Index 0=false
    Write Parameter Index 1=true
   )
   /SetVariables(
    Particles.RibbonColor:LinearColor=&MultiplyLinearColor(
     Linear Color=$Particles.Color
     Scale Factor=&MakeFloat4FromVectorAndFloat(
      "Vector (RGB)"=<1 1 1>
      "Float (Alpha)"=&FloatFromCurve(FloatCurve=DataInterface() {propertyValues {Curve {keys [
           {arriveTangent=16.4746, interpMode=RCIM_Cubic, leaveTangent=16.4746, tangentMode=RCTM_User, time=0, value=0}
           <at 0.15 1>
           {interpMode=RCIM_Cubic, leaveTangent=-1.8182, time=0.5, value=1}
           {interpMode=RCIM_Cubic, leaveTangent=-3.9348, time=0.75, value=0}
          ]}}})
     )
    )
    Particles.RibbonWidth=&FloatFromCurve(
     FloatCurve=DataInterface() {propertyValues {Curve {keys [
         {arriveTangent=16.9344, interpMode=RCIM_Cubic, leaveTangent=16.9344, tangentMode=RCTM_User, time=0, value=0}
         <at 0.15 1>
         <at 0.65 1 RCIM_Cubic>
         {arriveTangent=-3.9348, interpMode=RCIM_Cubic, leaveTangent=-3.9348, tangentMode=RCTM_User, time=0.85, value=0.25}
        ]}}}
     Scale Curve=3
    )
   ) {isSetParameters=true, module=SetVariables_#7}
  ]
  rendererClasses [&NiagaraMeshRendererProperties, &NiagaraRibbonRendererProperties]
  renderers [
   {rendererClass=&NiagaraMeshRendererProperties, rendererIndex=0}
   {rendererClass=&NiagaraRibbonRendererProperties, rendererIndex=1}
  ]
 }
 Constellation_Hands_Birth {
  emitterUpdateScript [
   &EmitterState(Life Cycle Mode=::NewEnumerator1, Loop Behavior=::NewEnumerator1, Loop Duration=0.5, UseLoopDelay=true)
   &SpawnBurst_Instantaneous(Spawn Count=6, Use Spawn Probability=false, UseLoopCountLimit=true)
  ]
  particleSpawnScript [
   &FN_SetSpawnGroup(Number of Spawn Groups=2)
   &InitializeParticle(
    Color=&LerpLinearColors(
     StartColor=$System.ColorA
     EndColor=$System.ColorB
     LerpFactor=&MakeFloatFromInt(Int=$Particles.SpawnGroup)
    )
    Lifetime Max=0.5
    Lifetime Min=0.4
    Lifetime Mode=::NewEnumerator1
    Mesh Scale=&MakeVector(
     X=&RandomRangeFloat(Minimum=0.5, Maximum=1)
     Y=&RandomRangeFloat(Minimum=0.5, Maximum=1)
     Z=&RandomRangeFloat(Minimum=0.5, Maximum=1)
    )
    Mesh Scale Mode=::NewEnumerator4
    UsePositionOffset=true
    Write Material Random=true
   )
   &InitialMeshOrientation(
    Rotation=&MakeVector(
     X=&RandomRangeFloat(Minimum=-0.2, Maximum=0.2)
     Y=&RandomRangeFloat(Minimum=-0.2, Maximum=0.2)
     Z=&RandomRangeFloat(Minimum=0, Maximum=1)
    )
    Use Orientation Vector=false
    Use Rotation Vector=true
   )
  ]
  particleUpdateScript [
   &ParticleState()
   /SetVariables(Particles.MeshOrientation=&MultiplyQuaternion(
     Quaternion A=&MultiplyQuaternion(
      Quaternion A=&FindQuatBetweenTwoVectors(From Vector=<0 0 -1>, To Vector=$System.DiveVelocityScale)
      Quaternion B=&FindQuatBetweenTwoVectors(From Vector=<1 0 0>, To Vector=$Engine.Owner.SystemYAxis)
     )
     Quaternion B=&MakeQuatFromAxisAngle(
      Axis=<0 0 1>
      AngleInDegrees=&Add_Float(
       A=&RandomRangeFloat(Minimum=-180, Maximum=180)
       B=&Lerp_Float(A=0, B=&RandomRangeFloat(Minimum=-90, Maximum=90), Alpha=$Particles.NormalizedAge)
      )
     )
    )) {enabled=false, isSetParameters=true, module=SetVariables_#5}
   &SkeletalMeshLocation(
    Bone / Socket Index=$Particles.SpawnGroup
    Bone Sampling Mode=::NewEnumerator4
    Inherit Velocity=false
    Interpolate Mesh Sampling=false
    MaskSpawnGroup=false
    Sampled Position Offset=&Lerp_Vector(
     A=&RandomRangeVector(Minimum=<-20 -20 -5>, Maximum=<20 20 5>)
     B=<0 0 0>
     Alpha=$Particles.NormalizedAge
    )
    Skeletal Mesh=$User.SkeletalMesh
   )
   /SetVariables(Particles.Position=&Add_Vector(A=Linked() {linkedVariable {name=Particles.Position, type=NiagaraPosition}}, B=<0 0 0>)) {enabled=false, isSetParameters=true, module=SetVariables_#6}
   &ScaleMeshSize(Scale Factor=&VectorFromCurve(
     VectorCurve=DataInterface() {propertyValues {
       XCurve {keys [
         {arriveTangent=2.9685, interpMode=RCIM_Cubic, time=0, value=0.25}
         {arriveTangent=-0.08, interpMode=RCIM_Cubic, leaveTangent=-0.08, time=0.33, value=1}
         {interpMode=RCIM_Cubic, leaveTangent=-1.8212, time=1, value=0.25}
        ]}
       YCurve {keys [
         {arriveTangent=2.9685, interpMode=RCIM_Cubic, time=0, value=0.25}
         {arriveTangent=-0.08, interpMode=RCIM_Cubic, leaveTangent=-0.08, time=0.33, value=1}
         {interpMode=RCIM_Cubic, leaveTangent=-1.8212, time=1, value=0.25}
        ]}
       ZCurve {keys [
         {arriveTangent=2.9685, interpMode=RCIM_Cubic, time=0, value=0.25}
         {arriveTangent=-0.08, interpMode=RCIM_Cubic, leaveTangent=-0.08, time=0.33, value=1}
         {interpMode=RCIM_Cubic, leaveTangent=-1.8212, time=1, value=0.25}
        ]}
      }}
     Scale Curve=<1.5 1.5 1.5>
    ))
   &ScaleColor(
    Scale Alpha=&FloatFromCurve(FloatCurve=DataInterface() {propertyValues {Curve {keys [<at 0 0 RCIM_Cubic>,<at 0.15 1>,<at 0.6 1 RCIM_Cubic>,<at 1 0 RCIM_Cubic>]}}})
    Scale RGB=&Multiply_VectorByFloat(
     Vector=<5 5 5>
     Float=&FloatFromCurve(FloatCurve=DataInterface() {propertyValues {Curve {keys [<at 0 0.3 RCIM_Cubic>,<at 0.15 1>,<at 0.6 1 RCIM_Cubic>,<at 1 0.3 RCIM_Cubic>]}}})
    )
    ScaleA=true
    ScaleRGB=true
   )
   &DynamicMaterialParameters(
    Index 1 Param 2=4
    Module.Param0WriteEnabled1=true
    Module.Param1WriteEnabled1=true
    Module.Param2WriteEnabled1=false
    Module.Param3WriteEnabled1=false
    Write Parameter Index 0=false
    Write Parameter Index 1=true
   )
   /SetVariables(
    Particles.RibbonColor:LinearColor=&MultiplyLinearColor(
     Linear Color=$Particles.Color
     Scale Factor=&MakeFloat4FromVectorAndFloat(
      "Vector (RGB)"=<1 1 1>
      "Float (Alpha)"=&FloatFromCurve(FloatCurve=DataInterface() {propertyValues {Curve {keys [
           {arriveTangent=16.4746, interpMode=RCIM_Cubic, leaveTangent=16.4746, tangentMode=RCTM_User, time=0, value=0}
           <at 0.15 1>
           {interpMode=RCIM_Cubic, leaveTangent=-1.8182, time=0.5, value=1}
           {interpMode=RCIM_Cubic, leaveTangent=-3.9348, time=0.75, value=0}
          ]}}})
     )
    )
    Particles.RibbonWidth=&FloatFromCurve(
     FloatCurve=DataInterface() {propertyValues {Curve {keys [
         {arriveTangent=16.9344, interpMode=RCIM_Cubic, leaveTangent=16.9344, tangentMode=RCTM_User, time=0, value=0}
         <at 0.15 1>
         <at 0.65 1 RCIM_Cubic>
         {arriveTangent=-3.9348, interpMode=RCIM_Cubic, leaveTangent=-3.9348, tangentMode=RCTM_User, time=0.85, value=0.25}
        ]}}}
     Scale Curve=3
    )
   ) {isSetParameters=true, module=SetVariables_#7}
  ]
  rendererClasses [&NiagaraMeshRendererProperties, &NiagaraRibbonRendererProperties]
  renderers [
   {rendererClass=&NiagaraMeshRendererProperties, rendererIndex=0}
   {rendererClass=&NiagaraRibbonRendererProperties, rendererIndex=1}
  ]
 }
 Lights {
  emitterUpdateScript [
   &EmitterState()
   &SpawnBurst_Instantaneous(Spawn Count=2, Use Spawn Probability=false, UseLoopCountLimit=false)
  ]
  particleSpawnScript [
   &FN_SetSpawnGroup(Number of Spawn Groups=2)
   &InitializeParticle(
    Color=&LerpLinearColors(
     StartColor=$System.ColorA
     EndColor=$System.ColorB
     LerpFactor=&MakeFloatFromInt(Int=$Particles.SpawnGroup)
    )
    Lifetime=3
    UsePositionOffset=false
    Write Material Random=true
   )
  ]
  particleUpdateScript [
   &ParticleState()
   &ScaleColor(
    Scale Alpha=&Lerp_Float(A=1, B=2, Alpha=&Sine(Normalized Angle=$Emitter.Age, Period=0.6, Scale=0.5, Bias=1))
    ScaleA=true
    ScaleRGB=true
   )
   &SkeletalMeshLocation(
    Bone / Socket Index=$Particles.SpawnGroup
    Bone Sampling Mode=::NewEnumerator4
    Inherit Velocity=false
    Interpolate Mesh Sampling=false
    MaskSpawnGroup=false
    Skeletal Mesh=$User.SkeletalMesh
   )
  ]
  rendererClasses [&NiagaraLightRendererProperties]
  renderers [{rendererClass=&NiagaraLightRendererProperties, rendererIndex=0}]
 }
 Ribbon_Hands {
  emitterUpdateScript [
   &EmitterState(Life Cycle Mode=::NewEnumerator1, Loop Duration=1, UseLoopDelay=false)
   &SpawnBurst_Instantaneous(Spawn Count=2, Use Spawn Probability=false, UseLoopCountLimit=false)
  ]
  particleSpawnScript [
   &FN_SetSpawnGroup(Number of Spawn Groups=2)
   &InitializeParticle(
    Color=&LerpLinearColors(
     StartColor=$System.ColorA
     EndColor=$System.ColorB
     LerpFactor=&MakeFloatFromInt(Int=$Particles.SpawnGroup)
    )
    Lifetime Max=1
    Lifetime Mode=::NewEnumerator1
    Sprite Size=&Multiply_Vector2D(
     A=<40 250>
     B=&Lerp_Vector2(A=<1 1>, B=<-1 1>, Alpha=&MakeFloatFromInt(Int=$Particles.SpawnGroup))
    )
    Sprite Size Mode=::NewEnumerator4
    UsePositionOffset=false
    Write Material Random=true
   )
   /SetVariables(Particles.UV Offset:NiagaraFloat=&Lerp_Float(
     A=&RandomRangeFloat(Minimum=0, Maximum=1)
     B=&RandomRangeFloat(Minimum=0, Maximum=1)
     Alpha=&MakeFloatFromInt(Int=$Particles.SpawnGroup)
    )) {isSetParameters=true}
  ]
  particleUpdateScript [
   &ParticleState()
   /SetVariables(Particles.SpriteAlignment=/Niagara/DynamicInputs/Vectors/NegateVector(Vector:Vector3f=$System.DiveVelocityScale)) {isSetParameters=true}
   &SkeletalMeshLocation(
    Bone / Socket Index=$Particles.SpawnGroup
    Bone Sampling Mode=::NewEnumerator4
    Inherit Velocity=false
    Interpolate Mesh Sampling=false
    MaskSpawnGroup=false
    Skeletal Mesh=$User.SkeletalMesh
   )
   &CameraOffset(Camera Offset Amount=-15)
   &ScaleSpriteSize(
    Scale Factor=&Multiply_Vector2D(
     A=&Lerp_Vector2(A=<1 1>, B=<0.75 1.5>, Alpha=$System.FN_Integrator001.Value)
     B=&MakeVector2D(
      X=&FloatFromCurve(
       FloatCurve=DataInterface() {propertyValues {Curve {keys [{arriveTangent=2.1688, interpMode=RCIM_Cubic, time=0, value=0.33}, <at 1.5 1 RCIM_Cubic>]}}}
       CurveIndex=$Emitter.Age
      )
      Y=&FN_FrontEndFloatModifier(Gameplay=1, Front End=0.75)
     )
    )
    Scale Sprite Size Mode=::NewEnumerator2
   )
   &ScaleColor(
    Scale Alpha=&FloatFromCurve(
     FloatCurve=DataInterface() {propertyValues {Curve {keys [{interpMode=RCIM_Cubic, tangentMode=RCTM_User, time=0, value=0}, {interpMode=RCIM_Cubic, tangentMode=RCTM_User, time=0.5, value=1}]}}}
     CurveIndex=$Emitter.Age
     Scale Curve=0.25
    )
    Scale RGB=<0.5 0.5 0.5>
    ScaleA=true
    ScaleRGB=true
   )
   &FN_Sprite_Attributes(
    Color=&MultiplyLinearColor(
     Linear Color=$Particles.Initial.Color
     Scale Factor=&MakeFloat4FromVectorAndFloat(
      "Vector (RGB)"=<0.75 0.75 0.75>
      "Float (Alpha)"=&FloatFromCurve(
       FloatCurve=DataInterface() {propertyValues {Curve {keys [{interpMode=RCIM_Cubic, tangentMode=RCTM_User, time=0, value=0}, {interpMode=RCIM_Cubic, tangentMode=RCTM_User, time=2, value=1}]}}}
       CurveIndex=$Emitter.Age
       Scale Curve=0.6
      )
     )
    )
    Create Color Attribute=true
    Write to Module Namespace=true
   )
   &DynamicMaterialParameters(
    Index 0 Param 1=&FloatFromCurve(
     FloatCurve=DataInterface() {propertyValues {Curve {keys [<at 0 1 RCIM_Cubic>, {interpMode=RCIM_Cubic, leaveTangent=-0.6543, time=1.5, value=0.1}]}}}
     CurveIndex=$Emitter.Age
    )
    Index 0 Param 2=&FloatFromCurve(
     FloatCurve=DataInterface() {propertyValues {Curve {keys [<at 0 1 RCIM_Cubic>, {interpMode=RCIM_Cubic, leaveTangent=-0.6543, time=1.5, value=0.33}]}}}
     CurveIndex=$Emitter.Age
    )
    Index 0 Param 3=$Particles.Initial.UV Offset
    Index 1 Param 1=&FloatFromCurve(
     FloatCurve=DataInterface() {propertyValues {Curve {keys [<at 0 1 RCIM_Cubic>, {interpMode=RCIM_Cubic, leaveTangent=-0.6543, time=1.5, value=0}]}}}
     CurveIndex=$Emitter.Age
    )
    Index 1 Param 2=1.5
    Index 1 Param 3=$Particles.Initial.UV Offset
    Module.Param0WriteEnabled=true
    Module.Param0WriteEnabled1=true
    Module.Param1WriteEnabled=true
    Module.Param1WriteEnabled1=true
    Module.Param2WriteEnabled=false
    Module.Param2WriteEnabled1=false
    Module.Param3WriteEnabled=false
    Module.Param3WriteEnabled1=false
    Write Parameter Index 1=true
   )
  ]
  rendererClasses [&NiagaraSpriteRendererProperties]
  renderers [
   {rendererClass=&NiagaraSpriteRendererProperties}
   {rendererClass=&NiagaraSpriteRendererProperties, rendererIndex=1}
  ]
 }
 Smoke_Hands {
  emitterUpdateScript [
   &EmitterState(Life Cycle Mode=::NewEnumerator1, Loop Duration=1, UseLoopDelay=false)
   &SpawnRate(SpawnRate=&Lerp_Float(A=6, B=12, Alpha=$System.FN_Integrator001.Value), Use Spawn Probability=true)
   &SpawnBurst_Instantaneous(Spawn Count=2, Use Spawn Probability=false, UseLoopCountLimit=true)
  ]
  particleSpawnScript [
   &FN_SetSpawnGroup(Number of Spawn Groups=2)
   &InitializeParticle(
    Color=&LerpLinearColors(
     StartColor=$System.ColorA
     EndColor=$System.ColorB
     LerpFactor=&MakeFloatFromInt(Int=$Particles.SpawnGroup)
    )
    Lifetime=&Lerp_Float(A=2, B=1.25, Alpha=$System.FN_Integrator001.Value)
    Sprite Rotation Mode=::NewEnumerator2
    Sprite Size Mode=::NewEnumerator1
    Uniform Sprite Size Max=100
    Uniform Sprite Size Min=80
    UsePositionOffset=false
    Write Material Random=true
   )
  ]
  particleUpdateScript [
   &ParticleState()
   &ScaleSpriteSize(Scale Factor=&Vector2DFromFloat(Value=&FloatFromCurve(FloatCurve=DataInterface() {propertyValues {Curve {keys [{arriveTangent=1.5832, time=0, value=0.6}, <at 0.2 0.85>, {leaveTangent=0.5893, time=1, value=1}]}}})))
   &SkeletalMeshLocation(
    Bone / Socket Index=$Particles.SpawnGroup
    Bone Sampling Mode=::NewEnumerator4
    Inherit Velocity=false
    Interpolate Mesh Sampling=false
    MaskSpawnGroup=false
    Sampled Position Offset=&Multiply_VectorByFloat(
     Vector=$System.DiveVelocityScale
     Float=&FloatFromCurve(
      FloatCurve=DataInterface() {propertyValues {Curve {keys [
          {arriveTangent=0.5515, interpMode=RCIM_Cubic, leaveTangent=0.5515, tangentMode=RCTM_User, time=0, value=-0.1}
          {arriveTangent=1.573, interpMode=RCIM_Cubic, leaveTangent=1.573, tangentMode=RCTM_User, time=1, value=1}
         ]}}}
      Scale Curve=$System.DiveOffset
     )
    )
    Skeletal Mesh=$User.SkeletalMesh
   )
   &ScaleColor(
    Scale Alpha=&FloatFromCurve(
     FloatCurve=DataInterface() {propertyValues {Curve {keys [<at 0 0 RCIM_Cubic>,<at 0.15 1>,<at 0.6 1 RCIM_Cubic>,<at 1 0 RCIM_Cubic>]}}}
     Scale Curve=0.5
    )
    Scale RGB=<0.5 0.5 0.5>
    ScaleA=true
    ScaleRGB=true
   )
   &CameraOffset(Camera Offset Amount=-15)
   &DynamicMaterialParameters(
    Index 0 Param 2=&FloatFromCurve(FloatCurve=DataInterface() {propertyValues {Curve {keys [
         {interpMode=RCIM_Cubic, tangentMode=RCTM_User, time=0.4, value=0.1}
         {arriveTangent=1.4285, interpMode=RCIM_Cubic, leaveTangent=1.4285, tangentMode=RCTM_User, time=1, value=1}
        ]}}})
    Module.Param0WriteEnabled=true
    Module.Param1WriteEnabled=true
    Module.Param2WriteEnabled=false
    Module.Param3WriteEnabled=false
   )
  ]
  rendererClasses [&NiagaraSpriteRendererProperties]
  renderers [{rendererClass=&NiagaraSpriteRendererProperties}]
 }
 Stars {
  emitterSpawnScript [/SetVariables(Emitter.StarScale=0.75) {isSetParameters=true}]
  emitterUpdateScript [
   &EmitterState(Life Cycle Mode=::NewEnumerator1, Loop Duration=1, UseLoopDelay=false)
   &SpawnRate(SpawnRate=&Lerp_Float(A=20, B=40, Alpha=$System.FN_Integrator001.Value), Use Spawn Probability=true)
  ]
  particleSpawnScript [
   &FN_SetSpawnGroup(Number of Spawn Groups=2, Set Ribbon ID and Ribbon Link Order=true)
   /SetVariables(Particles.VisibilityTag=&SelectIntFromArray(
     Direct Array Index=&RandomRangeInt(Minimum=0, Maximum=&MakeCustomIntFromBool(Bool=$User.bIsFrontEndPreview, True INT=0, False INT=9))
     Int Selection Array=DataInterface() {propertyValues {IntData [0,0,0,0,0,0,0,0,0,1]}}
    )) {isSetParameters=true}
   &InitializeParticle(
    Color=&LerpLinearColors(
     StartColor=$System.ColorA
     EndColor=$System.ColorB
     LerpFactor=&MakeFloatFromInt(Int=$Particles.SpawnGroup)
    )
    Lifetime=&Lerp_Float(A=1.25, B=0.75, Alpha=$System.FN_Integrator001.Value)
    Mesh Scale=&MakeVector(
     X=&RandomRangeFloat(Minimum=0.5, Maximum=2)
     Y=&RandomRangeFloat(Minimum=0.5, Maximum=2)
     Z=&RandomRangeFloat(Minimum=0.5, Maximum=2)
    )
    Mesh Scale Mode=::NewEnumerator4
    Sprite Size Max=<6 6>
    Sprite Size Min=<3 3>
    Sprite Size Mode=::NewEnumerator2
    UsePositionOffset=false
    Write Material Random=true
   )
   &CylinderLocation(
    Cylinder Height=10
    Cylinder Radius=&RandomRangeFloat(Minimum=100, Maximum=300)
    MaskSpawnGroup=false
    Offset=&Multiply_VectorByFloat(
     Vector=&NormalizeVector(Vector To Normalize=$System.DiveVelocityScale)
     Float=&Lerp_Float(A=200, B=400, Alpha=$System.FN_Integrator001.Value)
    )
    Offset Coordinate Space:ENiagaraCoordinateSpace=::World
    Offset Mode:ENiagara_OffsetMode=::NewEnumerator0
    Override Seed=false
    Surface Only=true
    Use Endcaps In Surface Only Mode=false
    UseOwnerScale=true
   )
   &AddVelocity(
    Coordinate Space=::World
    Scale Added Velocity=&VectorFromFloat(Value=&Lerp_Float(A=1, B=3, Alpha=$System.FN_Integrator001.Value))
    Velocity=&Multiply_VectorByFloat(Vector=$System.DiveVelocityScale, Float=-400)
   )
   &InitialMeshOrientation(
    Rotation=&MakeVector(
     X=&RandomRangeFloat(Minimum=-1, Maximum=1)
     Y=&RandomRangeFloat(Minimum=-1, Maximum=1)
     Z=&RandomRangeFloat(Minimum=-1, Maximum=1)
    )
    Use Orientation Vector=false
    Use Rotation Vector=true
   )
  ]
  particleUpdateScript [
   &ParticleState()
   &InheritSourceMovement(ApplyForce=false, ApplyPosition=true, ApplyPositionFalloff=false)
   &SolveForcesAndVelocity(Clamp Velocity=false, Limit Acceleration=false)
   &ScaleColor(
    Scale Alpha=&FloatFromCurve(FloatCurve=DataInterface() {propertyValues {Curve {keys [<at 0 0 RCIM_Cubic>,<at 0.15 1>,<at 0.8 1 RCIM_Cubic>,<at 1 0 RCIM_Cubic>]}}})
    Scale RGB=&Multiply_VectorByFloat(
     Vector=<5 5 5>
     Float=&FloatFromCurve(FloatCurve=DataInterface() {propertyValues {Curve {keys [<at 0 0.3 RCIM_Cubic>,<at 0.15 1>,<at 0.8 1 RCIM_Cubic>,<at 1 0.3 RCIM_Cubic>]}}})
    )
    ScaleA=true
    ScaleRGB=true
   )
   &FN_Sprite_Attributes(
    Color=&MultiplyLinearColor(Linear Color=$Particles.Color, Scale Factor=<1 1 1 1>)
    Create Color Attribute=true
    Create SpriteSize Attribute=true
    SpriteSize=&Multiply_Vector2D(A=$Particles.Initial.SpriteSize, B=&Lerp_Vector2(A=<1 1>, B=<1 10>, Alpha=$System.FN_Integrator001.Value))
    Write to Module Namespace=true
   )
   &ScaleMeshSize(Scale Factor=&VectorFromCurve(
     VectorCurve=DataInterface() {propertyValues {
       XCurve {keys [
         {arriveTangent=2.1662, interpMode=RCIM_Cubic, leaveTangent=2.1662, tangentMode=RCTM_User, time=0, value=0.5}
         <at 0.5 1 RCIM_Cubic>
         {arriveTangent=-1.8212, interpMode=RCIM_Cubic, leaveTangent=-1.8212, tangentMode=RCTM_User, time=1, value=0.5}
        ]}
       YCurve {keys [
         {arriveTangent=2.1662, interpMode=RCIM_Cubic, leaveTangent=2.1662, tangentMode=RCTM_User, time=0, value=0.5}
         <at 0.5 1 RCIM_Cubic>
         {arriveTangent=-1.8212, interpMode=RCIM_Cubic, leaveTangent=-1.8212, tangentMode=RCTM_User, time=1, value=0.5}
        ]}
       ZCurve {keys [
         {arriveTangent=2.1662, interpMode=RCIM_Cubic, leaveTangent=2.1662, tangentMode=RCTM_User, time=0, value=0.5}
         <at 0.5 1 RCIM_Cubic>
         {arriveTangent=-1.8212, interpMode=RCIM_Cubic, leaveTangent=-1.8212, tangentMode=RCTM_User, time=1, value=0.5}
        ]}
      }}
     Scale Curve=&Lerp_Vector(A=<1 1 1>, B=<0.6 0.6 1.2>, Alpha=$System.FN_Integrator001.Value)
    ))
   /SetVariables(Particles.RibbonColor:LinearColor=&MultiplyLinearColor(
     Linear Color=$Particles.Color
     Scale Factor=&MakeFloat4FromVectorAndFloat("Vector (RGB)"=<1 1 1>, "Float (Alpha)"=0.5)
    )) {isSetParameters=true}
  ]
  rendererClasses [&NiagaraSpriteRendererProperties, &NiagaraMeshRendererProperties]
  renderers [
   {rendererClass=&NiagaraSpriteRendererProperties}
   {rendererClass=&NiagaraSpriteRendererProperties, rendererIndex=1}
   {rendererClass=&NiagaraMeshRendererProperties, rendererIndex=2}
  ]
 }
}
summary {
 systemName=NS_Contrail_Division
 emitters [
  {emitterName=Smoke_Hands, rendererClasses [&NiagaraSpriteRendererProperties]}
  {
   emitterName=Constellation_Hands
   rendererClasses [&NiagaraMeshRendererProperties, &NiagaraRibbonRendererProperties]
  }
  {emitterName=Lights, rendererClasses [&NiagaraLightRendererProperties]}
  {emitterName=Ribbon_Hands, rendererClasses [&NiagaraSpriteRendererProperties]}
  {emitterName=Stars, rendererClasses [&NiagaraSpriteRendererProperties, &NiagaraMeshRendererProperties]}
  {
   emitterName=Constellation_Hands_Birth
   rendererClasses [&NiagaraMeshRendererProperties, &NiagaraRibbonRendererProperties]
  }
 ]
 userVariables [
  User.bIsFrontEnd {type=NiagaraBool}
  User.bIsFrontEndPreview {type=NiagaraBool, defaultValue=true}
  User.ParaGlideLeanAlpha {type=NiagaraFloat, defaultValue=0, description=""}
  User.RibbonIDOffset {type=NiagaraInt32}
  User.SkeletalMesh {
   type=NiagaraDataInterfaceSkeletalMesh
   defaultValue {
    struct=NiagaraExt_VariableValue_DataInterface
    value {dataInterface=&self:NiagaraDataInterfaceSkeletalMesh_0}
   }
  }
 ]
}
